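# Translucent unlit sheen material for Niagara mesh particles on the Dojo male one-handed pickaxe.
# A UV gradient, slanted by Skew, feeds MF_Sheen, whose mask position sweeps over particle relative time with adjustable width.
# The sheen's linear output is reflected, scaled by Gap and rounded into a hard-edged cut that multiplies the binary mask.
# Particle alpha scales that mask, while particle color passes through an emissive adjustment function.
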Material /BRCosmetics/Weapons/FORT_Melee/Pickaxe_Dojo_Male_1H/FX/M_Pickaxe_Dojo_M_1h_Sheen
graph {
 properties {
  bUsedWithNiagaraMeshParticles=true
  blendMode=BLEND_Translucent
  pixelDepthOffsetMode=PDOM_Legacy
  shadingModel=MSM_Unlit
  usedShadingModels=Unlit
 }
 nodes [
  n1 = ParticleColor()
  MaterialFunctionCall(Emissive=@n1.RGB) {materialFunction=&MF_FX_EmissiveAdjustments}
  n2 = TextureCoordinate()
  n3 = ComponentMask(None=@n2) {g=true}
  n4 = ScalarParameter() {defaultValue=0.1, parameterName=Skew, sliderMax=1}
  n5 = Multiply(A=OneMinus(None=ComponentMask(None=@n2) {r=true}), B=@n4)
  n6 = MaterialFunctionCall!(
   Gradient=Divide(A=Subtract(A=@n3, B=@n5), B=Add(B=@n4) {constA=1})
   Mask Position=OneMinus(None=ParticleRelativeTime())
   Mask Size=ScalarParameter() {defaultValue=0.15, parameterName=Sheen Width}
  ) {materialFunction=&/MF_Sheen}
  n7 = MaterialFunctionCall!(Gradient=@n6.Linear, Center Point=Constant() {r=0.8}) {materialFunction=/Game/Effects/Fort_Effects/MaterialFunctions/MF_GradientReflected}
  n8 = Saturate(None=Divide(A=OneMinus(None=@n7.Result), B=ScalarParameter() {defaultValue=0.2, parameterName=Gap}))
  Multiply(A=@n1.A, B=Multiply(A=Round(None=@n8), B=@n6.Binary))
 ]
}
