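# Unlit translucent Niagara sprite material that draws a procedural, texture-free sparkle for the Dojo pickaxe FX.
# Centered UVs and an eighth-turn rotated copy are computed per vertex through a VertexInterpolator.
# The rotated pair's absolute values are raised to Curvature and summed into a star shape, with Softness controlling the edge.
# Particle color feeds an emissive-adjustment function, and particle alpha scales the star mask.
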Material /BRCosmetics/Weapons/FORT_Melee/Pickaxe_Dojo_Male_1H/FX/M_Pickaxe_Dojo_Sparkle
graph {
 properties {bUsedWithNiagaraSprites=true, blendMode=BLEND_Translucent, pixelDepthOffsetMode=PDOM_Legacy, shadingModel=MSM_Unlit, usedShadingModels=Unlit}
 nodes [
  n1 = ParticleColor()
  MaterialFunctionCall(Emissive=@n1.RGB) {materialFunction=&MF_FX_EmissiveAdjustments}
  n2 = Divide(A=Subtract(A=TextureCoordinate()) {constB=0.5})
  n3 = ComponentMask(None=@n2) {r=true}
  n4 = Constant() {r=0.125}
  n5 = Cosine(None=@n4)
  n6 = ComponentMask(None=@n2) {g=true}
  n7 = Sine(None=@n4)
  n8 = Add(A=Multiply(A=@n3, B=@n5), B=Multiply(A=@n6, B=@n7))
  n9 = Subtract(A=Multiply(A=@n5, B=@n6), B=Multiply(A=@n7, B=@n3))
  n10 = VertexInterpolator(VS=AppendVector(A=@n2, B=AppendVector(A=@n8, B=@n9)))
  n11 = ComponentMask(None=Power(Base=Abs(None=@n10.PS), Exp=ScalarParameter() {defaultValue=0.25, parameterName=Curvature})) {a=true, b=true}
  n12 = OneMinus(None=Saturate(None=MaterialFunctionCall(f2=@n11) {materialFunction=&AddComponents}.f2))
  n13 = ScalarParameter() {defaultValue=0.1, parameterName=Softness}
  Multiply(A=@n1.A, B=Saturate(None=Divide(A=@n12, B=@n13)))
 ]
}
