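# Physics asset for the one-handed Dojo male axe pickaxe cosmetic, giving its strap bone chain collision bodies and joint limits for physical motion.
# Small, uniform capsules cover seven strap bones (no strap6) and joint1.
# Constraints link them in sequence from strap1 through strap8 to joint1, with moderate swing and slight twist along the straps.
# The final strap8-to-joint1 link has tighter swing and a zero twist limit.
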PhysicsAsset /BRCosmetics/Weapons/FORT_Melee/Pickaxe_Dojo_Male_1H/Meshes/Dojo_Male_1h_Axe_Physics
bodies {
 joint1 [{shapeType=Capsule, length=1, radius=1, rotation=<rot 90 0 0>}]
 strap1 [{shapeType=Capsule, length=1, radius=1, center=<1.2531 0 0>, rotation=<rot 90 0 0>}]
 strap2 [{shapeType=Capsule, length=1, radius=1, center=<1.2531 0 0>, rotation=<rot 90 0 0>}]
 strap3 [{shapeType=Capsule, length=1, radius=1, center=<1.2531 0 0>, rotation=<rot 90 0 0>}]
 strap4 [{shapeType=Capsule, length=1, radius=1, center=<1.2531 0 0>, rotation=<rot 90 0 0>}]
 strap5 [{shapeType=Capsule, length=1, radius=1, center=<1.2531 0 0>, rotation=<rot 90 0 0>}]
 strap7 [{shapeType=Capsule, length=1, radius=1, rotation=<rot 90 0 0>}]
 strap8 [{shapeType=Capsule, length=1, radius=1, rotation=<rot 90 0 0>}]
}
constraints [
 {bone1Name=strap2, bone2Name=strap1, swing1LimitDegrees=25, swing2LimitDegrees=25, twistLimitDegrees=7}
 {bone1Name=strap3, bone2Name=strap2, swing1LimitDegrees=25, swing2LimitDegrees=25, twistLimitDegrees=7}
 {bone1Name=strap4, bone2Name=strap3, swing1LimitDegrees=25, swing2LimitDegrees=25, twistLimitDegrees=7}
 {bone1Name=strap5, bone2Name=strap4, swing1LimitDegrees=25, swing2LimitDegrees=25, twistLimitDegrees=7}
 {bone1Name=strap7, bone2Name=strap5, swing1LimitDegrees=25, swing2LimitDegrees=25, twistLimitDegrees=7}
 {bone1Name=strap8, bone2Name=strap7, swing1LimitDegrees=25, swing2LimitDegrees=25, twistLimitDegrees=7}
 {bone1Name=joint1, bone2Name=strap8, swing1LimitDegrees=15, swing2LimitDegrees=15, twistLimitDegrees=0}
]
