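# Animation Blueprint for the Dojo Male one-handed axe pickaxe that adds secondary motion to its straps.
# A cached Copy Pose From Mesh feeds two branches, chosen by bRigidBodyEnabled: RigidBody physics with overridden gravity, or a trail controller fallback.
# The fallback additively rotates strap1 and strap2 before trailing the five-bone chain ending at strap5.
# Each update lowers those rotations' alpha in the frontend, using full strength otherwise.
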AnimBlueprint /BRCosmetics/Weapons/FORT_Melee/Pickaxe_Dojo_Male_1H/Meshes/Dojo_Male_1h_Axe_Skeleton_AnimBP
target_skeleton=&/Dojo_Male_1h_Axe_Skeleton
graphs [
 AnimGraph {nodes [
   "Transform (Modify) Bone - Bone: strap2"(Alpha=1, Translation="0,0,0", Rotation="-77.922,70.040,-67.705", Scale="1,1,1") {
    rotationMode=BMM_Additive
    rotationSpace=BCS_BoneSpace
    alphaBoolBlend {blendOption=Linear}
    boneToModify {boneName=strap2}
    rotation=<rot -77.922 70.0401 -67.7051>
   }
   n1 = Get Frontend() -> (Frontend:double)
   n2 = "Transform (Modify) Bone - Bone: strap1"(
    ComponentPose=Local To Component(LocalPose=Use cached pose 'In Pose'())
    Alpha=@n1
    Translation="0,0,0"
    Rotation="-55.887,-3.700,6.612"
    Scale="1,1,1"
   ) {
    rotationMode=BMM_Additive
    rotationSpace=BCS_BoneSpace
    alphaBoolBlend {blendOption=Linear}
    boneToModify {boneName=strap1}
    rotation=<rot -55.8866 -3.7 6.6123>
   }
   n3 = "Transform (Modify) Bone - Bone: strap2"(ComponentPose=@n2, Alpha=@n1, Translation="0,0,0", Rotation="-19.641,5.699,-1.090", Scale="1,1,1") {
    rotationMode=BMM_Additive
    rotationSpace=BCS_BoneSpace
    alphaBoolBlend {blendOption=Linear}
    boneToModify {boneName=strap2}
    rotation=<rot -19.641 5.699 -1.09>
   }
   n4 = Trail controller - Bone: strap5(ComponentPose=@n3, Alpha=1) {
    chainLength=5
    relaxationSpeedScale=4.5
    rotationLimits [{limitMax=<180 180 180>}, {limitMax=<180 180 180>}, {limitMax=<180 180 180>}, {limitMax=<180 180 180>}, {limitMax=<180 180 180>}]
    rotationOffsets [<0 0 0>, <0 0 0>, <0 0 0>, <0 0 0>, <0 0 0>]
    trailBone {boneName=strap5}
   }
   n5 = RigidBody(ComponentPose=Local To Component(LocalPose=Use cached pose 'In Pose'())) {
    bOverrideWorldGravity=true
    overridePhysicsAsset=&/Dojo_Male_1h_Axe_Physics
    alphaBoolBlend {blendOption=Linear}
    componentAppliedLinearAccClamp=<2000 2000 2000>
    componentLinearAccScale=<0.3 0.3 0.35>
    componentLinearVelScale=<0.3 0.3 0.35>
    overrideWorldGravity {z=-980}
   }
   n6 = Blend Poses by bool(BlendPose_0=Component To Local(ComponentPose=@n5), BlendPose_1=Component To Local(ComponentPose=@n4), bActiveValue=Get bRigidBodyEnabled()) {blendType=Linear}
   Output Pose(Result=Component To Local(ComponentPose=Local To Component(LocalPose=@n6))) {layerGroup=None}
   Save cached pose 'In Pose'(Pose=Copy Pose From Mesh()) {cacheName=In Pose}
  ]}
 EventGraph {nodes [
   Event BlueprintUpdateAnimation()
   > Set Frontend(Frontend:double=SelectFloat(A=0.38, B=1, bPickA=Get bIsInFrontEnd())) -> (Output_Get:double)
   TryGetPawnOwner()
  ]}
]
variables [Frontend {type=number}]
